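# UI-domain material named as a preview of the RT_CloudMask_01 render target, part of the game-setup UI for the Volumetrics cloud compositing tool.
# Its graph holds a single Texture Sample that reads that render target as linear color.
# The graph shows no other processing, parameters or links to material outputs.
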Material /Volumetrics/Tools/CloudCompositing/Blueprints/GameSetup/UI/RT_CloudMask_01_Preview
graph {
 properties {materialDomain=MD_UI, pixelDepthOffsetMode=PDOM_Legacy, shadingModel=MSM_Unlit, usedShadingModels=Unlit}
 nodes [TextureSample() {samplerType=SAMPLERTYPE_LinearColor, texture=/Volumetrics/Tools/CloudCompositing/RenderTargets/RT_CloudMask_01}]
}
